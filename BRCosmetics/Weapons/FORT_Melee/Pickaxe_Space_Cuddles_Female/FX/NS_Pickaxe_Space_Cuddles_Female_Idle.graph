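# Idle Niagara effect for the Space Cuddles Female pickaxe: a flickering thruster with electric arcs, attached to skeletal mesh sockets.
# Burst-spawned meshes form flame, main arc and two-socket glow, flickering through random scale, while a light pulses along a curve.
# Rate-spawned short-lived arc meshes follow their socket, and sprite embers shoot in a cone along socket orientation.
# Two user color themes tint everything through material parameters, ember color and light color.
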NiagaraSystem /BRCosmetics/Weapons/FORT_Melee/Pickaxe_Space_Cuddles_Female/FX/NS_Pickaxe_Space_Cuddles_Female_Idle
dependencies {
 usedDataInterfaces [&NiagaraDataInterfaceCurve, &NiagaraDataInterfaceColorCurve]
 usedDynamicInputs []
 usedModules [
  &SystemState
  &self:SystemSpawnScript.SystemScriptSource.SystemScriptGraph.NiagaraNodeAssignment_0.SetVariables_#1
 ]
}
emitters {
 Arc_Main {
  emitterUpdateScript [
   &EmitterState(Inactive Response=::NewEnumerator1, Life Cycle Mode=::NewEnumerator1, Loop Duration=1, UseLoopDelay=false)
   &SpawnBurst_Instantaneous(Spawn Count=1, Use Spawn Probability=false)
  ]
  particleSpawnScript [
   &InitializeParticle(
    Color=<rgba 10 10 10 5>
    Mesh Scale=<0.4 0.005 0.005>
    Mesh Scale Mode=::NewEnumerator4
    Position Offset=<50 0 0>
    Sprite Rotation Mode=::NewEnumerator2
    UsePositionOffset=false
    Write Material Random=true
   ) {module=InitializeParticle001}
   &SkeletalMeshLocation(
    Apply Sampled Orientation to Mesh Particles=true
    Bone / Socket Index=1
    Inherit Velocity=false
    Interpolate Mesh Sampling=false
    MaskSpawnGroup=false
    Mesh Sampling Type=::NewEnumerator5
    Sampled Position Offset=<3 0 0>
    Sampled Position Offset Coordinate Space:ENiagaraMeshTransforms=::NewEnumerator3
    Skeletal Mesh=$User.Skeletal Mesh
    Socket Sampling Mode=::NewEnumerator5
   )
   &DynamicMaterialParameters(Module.Param0WriteEnabled=true, Module.Param1WriteEnabled=true, Module.Param2WriteEnabled=true, Module.Param3WriteEnabled=true)
   /SetVariables(
    Particles.DynamicMaterialParameter1:Vector4f=&MakeVector4_FromLinearColor(LinearColor=$User.ColorThemeA)
    Particles.DynamicMaterialParameter2:Vector4f=&MakeVector4_FromLinearColor(LinearColor=$User.ColorThemeB)
   ) {isSetParameters=true}
  ]
  particleUpdateScript [
   &ParticleState()
   &ScaleMeshSize(Scale Factor=&RandomRangeVector(Minimum=<0.8 0.8 0.8>, Maximum=<1 1 1>, Evaluation Type=::NewEnumerator1))
  ]
  rendererClasses [&NiagaraMeshRendererProperties]
  renderers [{rendererClass=&NiagaraMeshRendererProperties, rendererIndex=0}]
 }
 Arcs {
  emitterUpdateScript [
   &EmitterState()
   &SpawnRate(Spawn Probability=0.6, SpawnRate=15, Use Spawn Probability=true)
  ]
  particleSpawnScript [
   /Niagara/Modules/Spawn/Initialization/InitializeParticle(
    Color=<rgba 10 10 10 1>
    Lifetime=&RandomRangeFloat(Minimum=0.1, Maximum=0.25)
    Mesh Scale=&RandomRangeVector(Minimum=<0.3 0.1 0.1>, Maximum=<1.5 0.25 0.25>)
    Module.Write Color=true
    Module.Write Lifetime=true
    Module.Write Mass=true
    Module.Write Material Random=true
    Module.Write Position=false
    Module.Write Scale=true
    Module.Write SpriteRotation=false
    Module.Write SpriteSize=false
    Position=&SimulationPosition()
   )
   &InitialMeshOrientation(
    Orientation Vector=<0 1 0>
    Rotation=&MakeVector(X=&RandomRangeFloat(Minimum=-1, Maximum=1), Y=0, Z=0)
    Use Orientation Vector=true
    Use Rotation Vector=true
   )
   &DynamicMaterialParameters(
    Index 0 Param 1=1.5
    Index 0 Param 2=&RandomRangeFloat(Minimum=0.1, Maximum=0.2)
    Index 0 Param 3=15
    Index 0 Param 4=1
    Module.Param0WriteEnabled=true
    Module.Param1WriteEnabled=true
    Module.Param2WriteEnabled=true
    Module.Param3WriteEnabled=true
   )
   /SetVariables(
    Particles.DynamicMaterialParameter1:Vector4f=&MakeVector4_FromLinearColor(LinearColor=$User.ColorThemeA)
    Particles.DynamicMaterialParameter2:Vector4f=&MakeVector4_FromLinearColor(LinearColor=$User.ColorThemeB)
   ) {isSetParameters=true}
  ]
  particleUpdateScript [
   &ParticleState()
   &SkeletalMeshLocation(
    Bone / Socket Index=1
    Inherit Velocity=true
    MaskSpawnGroup=false
    Mesh Sampling Type=::NewEnumerator5
    Sampled Position Offset Coordinate Space:ENiagaraMeshTransforms=::NewEnumerator3
    Skeletal Mesh=$User.Skeletal Mesh
    Socket Sampling Mode=::NewEnumerator5
    VelocitySpeedLimit=false
   )
   &SolveForcesAndVelocity(Clamp Velocity=false, Limit Acceleration=false) {enabled=false}
   &CameraOffset(Camera Offset Amount=5) {enabled=false}
   &ScaleColor(
    Scale Mode=::NewEnumerator1
    Scale RGBA=&Vector4FromCurve(Vector4Curve=DataInterface() {propertyValues {
       AlphaCurve {keys [<at 0.0008 0>,<at 0.1906 1>,<at 0.9069 1>,<at 1.0267 0>]}
       BlueCurve {keys [<at 0.0527 0.9502>,<at 1.0052 2>]}
       GreenCurve {keys [<at 0.0527 1>,<at 1.0052 1.1196>]}
       RedCurve {keys [<at 0.0527 0.165>,<at 1.0052 0.2058>]}
      }})
    ScaleRGBA=true
   ) {enabled=false}
   &ScaleMeshSize(Scale Factor=&VectorFromFloat(Value=&FloatFromCurve(FloatCurve=DataInterface() {propertyValues {Curve {keys [
          {arriveTangent=12.2006, interpMode=RCIM_Cubic, leaveTangent=3.0184, tangentMode=RCTM_User, time=0, value=0}
          {arriveTangent=2.1518, interpMode=RCIM_Cubic, leaveTangent=2.1518, time=0.2304, value=0.8427}
          {arriveTangent=-0.1643, interpMode=RCIM_Cubic, leaveTangent=-0.1643, time=0.4564, value=0.9821}
          {arriveTangent=-1.8066, interpMode=RCIM_Cubic, leaveTangent=-1.8066, time=0.7318, value=0.7603}
          {arriveTangent=-1.9497, interpMode=RCIM_Cubic, leaveTangent=-5.2827, tangentMode=RCTM_User, time=1, value=0}
         ]}}}))) {module=ScaleMeshSize001}
  ]
  rendererClasses [&NiagaraMeshRendererProperties]
  renderers [{rendererClass=&NiagaraMeshRendererProperties, rendererIndex=0}]
 }
 Embers {
  emitterUpdateScript [
   &EmitterState() {module=EmitterState001}
   &SpawnRate(Spawn Probability=0.6, SpawnRate=60, Use Spawn Probability=true)
  ]
  particleSpawnScript [
   /Niagara/Modules/Spawn/Initialization/V2/InitializeParticle(
    Color=<rgba 5 5 5 1>
    Lifetime=&UniformRangedFloat(Minimum=0.06, Maximum=0.166)
    Sprite Size=&Multiply_Vector2(
     A=<1 6>
     B=&Vector2DFromFloat(Value=&RandomRangeFloat(Minimum=0.1, Maximum=0.3))
    )
    Sprite Size Mode=::NewEnumerator4
    UsePositionOffset=false
    Write Material Random=true
   ) {module=InitializeParticle001}
   &SkeletalMeshLocation(
    Inherit Velocity=false
    MaskSpawnGroup=false
    Mesh Sampling Type=::NewEnumerator5
    Sampled Position Offset=<1 0 0>
    Sampled Position Offset Coordinate Space:ENiagaraMeshTransforms=::NewEnumerator3
    Skeletal Mesh=$User.Skeletal Mesh
    Socket Sampling Mode=::NewEnumerator5
    Use Sampled Mesh Attribute as Ribbon ID=false
   )
   &SphereLocation(
    MaskSpawnGroup=false
    Non Uniform Scale=<1 0.25 1>
    Override Seed=false
    Sphere Radius=4
    Surface Only=false
    UseNonUniformScale=true
    UseOffset=false
    UseSphereOrientation=false
   )
   &AddVelocityInCone(
    Cone Angle=25
    Cone Axis=&RotateVectorByQuaternion(Vector To Rotate=<1 0 0>, Quaternion=$Output.SkeletalMeshLocation.SampledRotation)
    Cone Axis Coordinate Space=::Simulation
    Module.Use Velocity Falloff On Cone Axis=false
    Velocity Strength=100
   )
  ]
  particleUpdateScript [
   &ParticleState()
   &ScaleVelocity(Coordinate Space:ENiagaraCoordinateSpace=::Simulation) {module=ScaleVelocity001}
   &Drag(Drag=3, Use Linear Drag=true, Use Rotational Drag=false)
   &SolveForcesAndVelocity(Clamp Velocity=false, Limit Acceleration=false)
   &ScaleSpriteSize(Scale Factor=&MakeVector2D(X=1, Y=1))
   &ScaleColor(
    Scale Alpha=&FloatFromCurve(FloatCurve=DataInterface() {propertyValues {Curve {keys [{interpMode=RCIM_Cubic, tangentMode=RCTM_User, time=0.5, value=1}, {interpMode=RCIM_Cubic, tangentMode=RCTM_User, time=1, value=0}]}}})
    Scale RGB=&MakeVectorFromLinearColorRGB(Color=&LerpLinearColors(
      StartColor=$User.ColorThemeA
      EndColor=$User.ColorThemeB
      LerpFactor=&FloatFromCurve(FloatCurve=DataInterface() {propertyValues {Curve {keys [{interpMode=RCIM_Cubic, tangentMode=RCTM_User, time=0.5, value=1}, {interpMode=RCIM_Cubic, tangentMode=RCTM_User, time=1, value=0.1}]}}})
     ))
    ScaleA=true
    ScaleRGB=true
   )
  ]
  rendererClasses [&NiagaraSpriteRendererProperties]
  renderers [{rendererClass=&NiagaraSpriteRendererProperties}]
 }
 Thruster_Flame {
  emitterUpdateScript [
   &EmitterState(Inactive Response=::NewEnumerator1, Life Cycle Mode=::NewEnumerator1, Loop Duration=1, UseLoopDelay=false)
   &SpawnBurst_Instantaneous(Spawn Count=1, Use Spawn Probability=false)
  ]
  particleSpawnScript [
   &InitializeParticle(
    Color=<rgba 10 10 10 1>
    Mesh Scale=<0.1 0.2 0.2>
    Mesh Scale Mode=::NewEnumerator4
    Position Offset=<50 0 0>
    Sprite Rotation Mode=::NewEnumerator2
    UsePositionOffset=false
    Write Material Random=true
   ) {module=InitializeParticle001}
   &SkeletalMeshLocation(
    Apply Sampled Orientation to Mesh Particles=true
    Inherit Velocity=false
    Interpolate Mesh Sampling=false
    MaskSpawnGroup=false
    Mesh Sampling Type=::NewEnumerator5
    Sampled Position Offset=<0 0 -0.1>
    Sampled Position Offset Coordinate Space:ENiagaraMeshTransforms=::NewEnumerator3
    Skeletal Mesh=$User.Skeletal Mesh
    Socket Sampling Mode=::NewEnumerator5
   )
   &DynamicMaterialParameters(Module.Param0WriteEnabled=true, Module.Param1WriteEnabled=true, Module.Param2WriteEnabled=true, Module.Param3WriteEnabled=true)
   /SetVariables(
    Particles.DynamicMaterialParameter1:Vector4f=&MakeVector4_FromLinearColor(LinearColor=$User.ColorThemeA)
    Particles.DynamicMaterialParameter2:Vector4f=&MakeVector4_FromLinearColor(LinearColor=$User.ColorThemeB)
   ) {isSetParameters=true}
  ]
  particleUpdateScript [
   &ParticleState()
   &ScaleMeshSize(Scale Factor=&RandomRangeVector(Minimum=<0.75 0.8 0.8>, Maximum=<1 1 1>, Evaluation Type=::NewEnumerator1))
   &CameraOffset(Camera Offset Amount=1)
  ]
  rendererClasses [&NiagaraMeshRendererProperties]
  renderers [{rendererClass=&NiagaraMeshRendererProperties, rendererIndex=0}]
 }
 Thruster_Glow001 {
  emitterUpdateScript [
   &EmitterState(Inactive Response=::NewEnumerator1, Life Cycle Mode=::NewEnumerator1, Loop Duration=1, UseLoopDelay=false)
   &SpawnBurst_Instantaneous(Spawn Count=2, Use Spawn Probability=false)
  ]
  particleSpawnScript [
   &FN_SetSpawnGroup(Number of Spawn Groups=2)
   &InitializeParticle(
    Color=<rgba 5 5 5 0.5>
    Mesh Scale=&MaskVectorBySpawnGroup(Masked Vector=<0.55 0.04 0.04>, Passthrough Vector=<0.15 0.1 0.1>, Spawn Group Mask=1)
    Mesh Scale Mode=::NewEnumerator4
    Position Offset=<50 0 0>
    Sprite Rotation Mode=::NewEnumerator2
    UsePositionOffset=false
    Write Material Random=true
   ) {module=InitializeParticle001}
   &SkeletalMeshLocation(
    Apply Sampled Orientation to Mesh Particles=true
    Bone / Socket Index=&SpawnGroup()
    Inherit Velocity=false
    Interpolate Mesh Sampling=false
    MaskSpawnGroup=false
    Mesh Sampling Type=::NewEnumerator5
    Sampled Position Offset Coordinate Space:ENiagaraMeshTransforms=::NewEnumerator3
    Skeletal Mesh=$User.Skeletal Mesh
    Socket Sampling Mode=::NewEnumerator5
   )
   &DynamicMaterialParameters(Module.Param0WriteEnabled=true, Module.Param1WriteEnabled=true, Module.Param2WriteEnabled=true, Module.Param3WriteEnabled=true)
   /SetVariables(
    Particles.DynamicMaterialParameter1:Vector4f=&MakeVector4_FromLinearColor(LinearColor=$User.ColorThemeA)
    Particles.DynamicMaterialParameter2:Vector4f=&MakeVector4_FromLinearColor(LinearColor=$User.ColorThemeA)
   ) {isSetParameters=true}
  ]
  particleUpdateScript [
   &ParticleState()
   &ScaleColor(ScaleA=true, ScaleRGB=true)
   &ScaleMeshSize(Scale Factor=&RandomRangeVector(Minimum=<0.75 0.8 0.8>, Maximum=<1 1 1>, Evaluation Type=::NewEnumerator1))
  ]
  rendererClasses [&NiagaraMeshRendererProperties]
  renderers [{rendererClass=&NiagaraMeshRendererProperties, rendererIndex=0}]
 }
 Thruster_Light {
  emitterUpdateScript [
   &EmitterState(Inactive Response=::NewEnumerator1, Life Cycle Mode=::NewEnumerator1, Loop Duration=1, UseLoopDelay=false)
   &SpawnBurst_Instantaneous(Spawn Count=1, Use Spawn Probability=false)
  ]
  particleSpawnScript [
   &InitializeParticle(Color=<rgba 5 5 5 1>, Position Mode=::NewEnumerator1, Position Offset=<50 0 0>, UsePositionOffset=false, Write Material Random=true) {module=InitializeParticle001}
   &SkeletalMeshLocation(
    Apply Sampled Orientation to Mesh Particles=true
    Bone / Socket Index=1
    Inherit Velocity=false
    Interpolate Mesh Sampling=false
    MaskSpawnGroup=false
    Mesh Sampling Type=::NewEnumerator5
    Sampled Position Offset Coordinate Space:ENiagaraMeshTransforms=::NewEnumerator3
    Skeletal Mesh=$User.Skeletal Mesh
    Socket Sampling Mode=::NewEnumerator5
   )
   /SetVariables(Particles.LightRadius=75) {isSetParameters=true}
   /SetVariables(Particles.LightPosition:NiagaraPosition=$Particles.Position) {isSetParameters=true}
  ]
  particleUpdateScript [
   &ParticleState()
   /SetVariables(Particles.LightColor=&MultiplyLinearColor(
     Linear Color=$User.ColorThemeB
     Scale Factor=&MakeVector4_FromFloat(Float=&Lerp_Float(
       A=0.4
       B=1
       Alpha=&FloatFromCurve(FloatCurve=DataInterface() {propertyValues {Curve {keys [
            <at 0 0 RCIM_Cubic>
            {arriveTangent=1.304, interpMode=RCIM_Cubic, leaveTangent=1.304, time=0.1, value=0.75}
            {arriveTangent=-1.2483, interpMode=RCIM_Cubic, leaveTangent=-1.2483, time=0.1681, value=0.2192}
            {arriveTangent=-0.9034, interpMode=RCIM_Cubic, leaveTangent=-0.9034, time=0.216, value=1}
            {arriveTangent=-0.1111, interpMode=RCIM_Cubic, leaveTangent=-0.1111, time=0.3, value=0.1}
            {arriveTangent=1.8356, interpMode=RCIM_Cubic, leaveTangent=1.8356, time=0.3516, value=0.6401}
            {arriveTangent=2.4257, interpMode=RCIM_Cubic, leaveTangent=2.4257, time=0.4219, value=0}
            {arriveTangent=0.2031, interpMode=RCIM_Cubic, leaveTangent=0.2031, time=0.5, value=1}
            {arriveTangent=-2.3742, interpMode=RCIM_Cubic, leaveTangent=-2.3742, time=0.5703, value=0.3538}
            {arriveTangent=-1.957, interpMode=RCIM_Cubic, leaveTangent=-1.957, time=0.6325, value=0.6853}
            {arriveTangent=-0.2171, interpMode=RCIM_Cubic, leaveTangent=-0.2171, time=0.7, value=0}
            {arriveTangent=1.2231, interpMode=RCIM_Cubic, leaveTangent=1.2231, time=0.7714, value=0.6552}
            {arriveTangent=1.1257, interpMode=RCIM_Cubic, leaveTangent=1.1257, time=0.8336, value=0.2634}
            {arriveTangent=-1.5833, interpMode=RCIM_Cubic, leaveTangent=-1.5833, time=0.9, value=1}
            <at 1 0 RCIM_Cubic>
           ]}}})
      ))
    )) {isSetParameters=true}
  ]
  rendererClasses [&NiagaraLightRendererProperties]
  renderers [{rendererClass=&NiagaraLightRendererProperties, rendererIndex=0}]
 }
}
summary {
 systemName=NS_Pickaxe_Space_Cuddles_Female_Idle
 emitters [
  {emitterName=Thruster_Flame, rendererClasses [&NiagaraMeshRendererProperties]}
  {emitterName=Thruster_Light, rendererClasses [&NiagaraLightRendererProperties]}
  {emitterName=Arc_Main, rendererClasses [&NiagaraMeshRendererProperties]}
  {emitterName=Arcs, rendererClasses [&NiagaraMeshRendererProperties]}
  {emitterName=Thruster_Glow001, rendererClasses [&NiagaraMeshRendererProperties]}
  {emitterName=Embers, rendererClasses [&NiagaraSpriteRendererProperties]}
 ]
 userVariables [
  User.ColorThemeA {type=LinearColor, description="", defaultValue {struct=LinearColor, value=<rgba 0.5 0 0.25 1>}}
  User.ColorThemeB {type=LinearColor, description="", defaultValue {struct=LinearColor, value=<rgba 1.5 0 1 1>}}
  User.Skeletal Mesh {
   type=NiagaraDataInterfaceSkeletalMesh
   defaultValue {
    struct=NiagaraExt_VariableValue_DataInterface
    value {dataInterface=&self:NiagaraDataInterfaceSkeletalMesh_0}
   }
  }
  User.SkeletalMesh {
   type=NiagaraDataInterfaceSkeletalMesh
   defaultValue {
    struct=NiagaraExt_VariableValue_DataInterface
    value {dataInterface=&self:NiagaraDataInterfaceSkeletalMesh_1}
   }
  }
 ]
}
