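# Physics asset for the female Space Cuddles two-handed pickaxe, defining collision bodies and joints for the handle and its dangling charm.
# One long capsule covers the handle, and small capsules cover the four charm bones, with the end one nearly spherical.
# Constraints chain the charm from the handle through dyn_charm_base to dyn_charm_03, each with the same tight swing and twist limit angles.
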PhysicsAsset /BRCosmetics/Weapons/FORT_Melee/Pickaxe_Space_Cuddles_Female/Meshes/Space_Cuddles_Female_Axe_Physics
bodies {
 Melee_TwoHanded_Handle [{shapeType=Capsule, length=93.2317, radius=1.9622, center=<1.0412 0 0>}]
 dyn_charm_01 [{shapeType=Capsule, length=2.568, radius=1.1693, center=<0.0612 -1.4457 -0.0301>, rotation=<rot 0.0746 3.5902 -91.1883>}]
 dyn_charm_02 [{shapeType=Capsule, length=1.9593, radius=1.0652, center=<0 -0.8178 -0.0172>, rotation=<rot 0.1244 5.9284 -85.3918>}]
 dyn_charm_03 [{shapeType=Capsule, length=0.002, radius=3.2948, center=<-0.0996 -1.1084 -0.0354>, rotation=<rot 0 0 -88.716>}]
 dyn_charm_base [{shapeType=Capsule, length=2.2023, radius=1.1658, center=<-0.2534 -0.4879 0.0717>, rotation=<rot -82.3193 13.6892 -127.4874>}]
}
constraints [
 {bone1Name=dyn_charm_02, bone2Name=dyn_charm_01, swing1LimitDegrees=10, swing2LimitDegrees=10, twistLimitDegrees=10}
 {bone1Name=dyn_charm_03, bone2Name=dyn_charm_02, swing1LimitDegrees=10, swing2LimitDegrees=10, twistLimitDegrees=10}
 {bone1Name=dyn_charm_01, bone2Name=dyn_charm_base, swing1LimitDegrees=10, swing2LimitDegrees=10, twistLimitDegrees=10}
 {bone1Name=dyn_charm_base, bone2Name=Melee_TwoHanded_Handle, swing1LimitDegrees=10, swing2LimitDegrees=10, twistLimitDegrees=10}
]
